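# Corrective Control Rig for the VelvetDeskCam male medium player body, adjusting deform and accessory bones to the pose.
# It runs optional debug controls, a shared DeformRig function, Parachute upper-arm offsets, BeltFix, then pose-reader knee and left-forearm correctives.
# BeltFix turns thigh swing into clamped weights, combined with crouch and DBNO variables, to offset belt, bag and grenade bones.
# The parent-constraint function (Constrains) is left unwired.
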ControlRigBlueprint /BRCosmetics/Characters/Player/Male/Medium/Bodies/M_MED_VelvetDeskCam/Meshes/M_MED_VelvetDeskCam_CtrlRig
graphs [
 Multiply_4_ContainedGraph {nodes [
   n1 = RigVMFunctionEntryNode()
   Multiply = &RigVMFunction_MathDoubleMul(A=@n1.A, B=@n1.B)
   RigVMFunctionReturnNode(Result=&RigVMFunction_MathDoubleMul(B=@n1.C, A=@Multiply.Result).Result)
  ]}
 New Function_ContainedGraph {nodes [
   RigVMFunctionEntryNode()
   > &RigUnit_ParentConstraint(
    AdvancedSettings=(InterpolationType=Average,RotationOrderForFilter=XZY)
    Child=(Type=Bone,Name="base_belt_area")
    Filter=(TranslationFilter=(bX=true,bY=true,bZ=true),RotationFilter=(bX=True,bY=True,bZ=True),ScaleFilter=(bX=True,bY=True,bZ=True))
    Parents=((Item=(Type=Bone,Name="thigh_r"),Weight=1))
    Weight=0.6
   )
   > RigVMFunctionReturnNode()
   &RigUnit_ParentConstraint(
    AdvancedSettings=(InterpolationType=Average,RotationOrderForFilter=XZY)
    Child=(Type=Bone,Name="base_bag_area")
    ExecutePin=()
    Parents=((Item=(Type=Bone,Name="thigh_l"),Weight=1))
    Weight=0.3
   )
  ]}
 New Function_ContainedGraph {nodes [
   n1 = RigVMFunctionEntryNode()
   Clamp = &RigVMFunction_MathIntClamp(Value=@n1.primAxis)
   RigVMFunction_MathBoolToInteger = &RigVMFunction_MathBoolToInteger(Value=@n1.onlyTwist)
   RigVMFunction_MathQuaternionSwingTwist = &RigVMFunction_MathQuaternionSwingTwist(
    TwistAxis=@n1.twistAxis
    Input=&RigUnit_GetRelativeTransformForItem(
     Child=(Type=Bone,Name="None")
     Parent=(Type=Bone,Name="None")
     bChildInitial=False
     bParentInitial=False
     Parent.Type=@n1.ParentType
     Child.Name=@n1.child
     Parent.Name=@n1.parent
    ).RelativeTransform.Rotation
   )
   RigVMFunction_MathQuaternionToEuler = &RigVMFunction_MathQuaternionToEuler(
    Value=RigVMDispatchNode(
     Values=((),())
     Index=@RigVMFunction_MathBoolToInteger.Result
     Values.0=@RigVMFunction_MathQuaternionSwingTwist.Swing
     Values.1=@RigVMFunction_MathQuaternionSwingTwist.Twist
    ) {template="DISPATCH_RigVMDispatch_SelectInt32(in Index,in Values,out Result)"}.Result
    RotationOrder=@n1.RotationOrder
   )
   DISPATCH_RigVMDispatch_SelectInt32 = RigVMDispatchNode(
    Values=(0,0,0)
    Index=@Clamp.Result
    Values.0=@RigVMFunction_MathQuaternionToEuler.Result.X
    Values.1=@RigVMFunction_MathQuaternionToEuler.Result.Y
    Values.2=@RigVMFunction_MathQuaternionToEuler.Result.Z
   ) {template="DISPATCH_RigVMDispatch_SelectInt32(in Index,in Values,out Result)"}
   Remap = &RigVMFunction_MathDoubleRemap(
    TargetMaximum=1
    TargetMinimum=0
    bClamp=False
    Value=@DISPATCH_RigVMDispatch_SelectInt32.Result
    SourceMaximum=@n1.max
    SourceMinimum=@n1.min
   )
   > RigVMFunctionReturnNode(ouputWeight=&RigVMFunction_MathDoubleClamp(Value=@Remap.Result).Result)
  ]}
 New Function_ContainedGraph {nodes [
   n1 = RigVMFunctionEntryNode()
   RigVMFunction_ControlFlowBranch = &RigVMFunction_ControlFlowBranch(Condition=@n1.DebugControls, ExecuteContext=@n1.ExecuteContext)
   RigVMFunctionReturnNode(ExecuteContext=@RigVMFunction_ControlFlowBranch.Completed)
   RigUnit_GetTransform = &RigUnit_GetTransform(Item=(Type=Control,Name="thigh_l_ctrl"), bInitial=False)
   RigUnit_SetTransform = &RigUnit_SetTransform(
    Item=(Type=Bone,Name="thigh_l")
    Weight=1
    bInitial=False
    Value=@RigUnit_GetTransform.Transform
    ExecutePin=@RigVMFunction_ControlFlowBranch.True
   )
   > &RigUnit_SetTransform(
    Item=(Type=Bone,Name="thigh_r")
    Weight=1
    bInitial=False
    Value=&RigUnit_GetTransform(Item=(Type=Control,Name="thigh_r_ctrl"), bInitial=False).Transform
   )
   > &RigUnit_SetTransform(
    Item=(Type=Bone,Name="calf_l")
    Weight=1
    bInitial=False
    Value=&RigUnit_GetTransform(Item=(Type=Control,Name="calf_l_ctrl"), bInitial=False).Transform
   )
   > &RigUnit_SetTransform(
    Item=(Type=Bone,Name="calf_r")
    Weight=1
    bInitial=False
    Value=&RigUnit_GetTransform(Item=(Type=Control,Name="calf_r_ctrl"), bInitial=False).Transform
   )
   > &RigUnit_SetTransform(
    Item=(Type=Bone,Name="hand_l")
    Weight=1
    bInitial=False
    Value=&RigUnit_GetTransform(Item=(Type=Control,Name="hand_l_ctrl"), bInitial=False).Transform
   )
  ]}
 New Function_ContainedGraph {nodes [
   Delta Rotation_4 = RigVMFunctionReferenceNode(
    ExecuteContext=()
    ParentType=Null
    RotationOrder=YZX
    child=thigh_r
    debugRotation=False
    max=90
    onlyTwist=False
    parent=thigh_r_null
    primAxis=2
    twistAxis=(X=-1,Y=0,Z=0)
   )
   > &RigUnit_ModifyTransforms(
    ItemToModify=((Item=(Type=Bone,Name="base_grenade_A"),Transform=(Rotation=(X=0,Y=0,Z=0.7071,W=0.7071),Translation=(X=10,Y=0,Z=0),Scale3D=(X=1,Y=1,Z=1))))
    WeightMinimum=0
    Weight=@"Delta Rotation_4".ouputWeight
   )
   RigVMFunctionEntryNode()
   > RigVMFunction_Sequence_1 = RigVMAggregateNode()
   Delta Rotation = RigVMFunctionReferenceNode(
    ParentType=Null
    RotationOrder=YZX
    child=thigh_r
    debugRotation=False
    max=-120
    min=5
    onlyTwist=False
    parent=thigh_r_null
    primAxis=2
    twistAxis=(X=-1,Y=0,Z=0)
    ExecuteContext=@RigVMFunction_Sequence_1.A
   )
   > &RigUnit_ModifyTransforms(
    ItemToModify="((Item=(Type=Bone,Name=\"base_belt_area\"),Transform=(Rotation=(X=-0.0243,Y=0.251,Z=-0.0886,W=0.9636),Translation=(X=0,Y=0,Z=0),Scale3D=(X=1,Y=1,Z=1))),(Item=(Type=Bone,Name=\"base_grenades_area\"),Transform=(Rotation=(X=0.0626,Y=0.1497,Z=0.1202,W=0.9794),Tran ...(truncated)"
    WeightMinimum=0
    Weight=@"Delta Rotation".ouputWeight
   )
   > RigVMVariableNode(Value=@"Delta Rotation".ouputWeight) {variable=legDeltaRot_SwingZup_r}
   > Delta Rotation_1 = RigVMFunctionReferenceNode(
    ParentType=Null
    RotationOrder=YZX
    child=thigh_r
    debugRotation=False
    max=90
    onlyTwist=False
    parent=thigh_r_null
    primAxis=2
    twistAxis=(X=-1,Y=0,Z=0)
   )
   > &RigUnit_ModifyTransforms(
    ItemToModify="((Item=(Type=Bone,Name=\"base_belt_area\"),Transform=(Rotation=(X=-0.0896,Y=0.2781,Z=0.3461,W=0.8915),Translation=(X=0,Y=0,Z=0),Scale3D=(X=1,Y=1,Z=1))),(Item=(Type=Bone,Name=\"base_grenades_area\"),Transform=(Rotation=(X=0.037,Y=0.0533,Z=0.0816,W=0.9945),Trans ...(truncated)"
    WeightMinimum=0
    Weight=@"Delta Rotation_1".ouputWeight
   )
   > Delta Rotation_1_1 = RigVMFunctionReferenceNode(
    ParentType=Null
    RotationOrder=ZXY
    child=thigh_r
    debugRotation=False
    max=-90
    onlyTwist=False
    parent=thigh_r_null
    primAxis=1
    twistAxis=(X=-1,Y=0,Z=0)
   )
   > &RigUnit_ModifyTransforms(
    ItemToModify="((Item=(Type=Bone,Name=\"base_belt_area\"),Transform=(Rotation=(X=0,Y=0.3079,Z=0,W=0.9514),Translation=(X=0,Y=0,Z=0),Scale3D=(X=1,Y=1,Z=1))),(Item=(Type=Bone,Name=\"base_grenade_B\"),Transform=(Rotation=(X=0,Y=0,Z=0.1961,W=0.9806),Translation=(X=0,Y=-0.5631,Z= ...(truncated)"
    WeightMinimum=0
    Weight=@"Delta Rotation_1_1".ouputWeight
   )
   > Delta Rotation_1_1_3 = RigVMFunctionReferenceNode(
    ParentType=Null
    RotationOrder=ZXY
    child=thigh_r
    debugRotation=False
    max=90
    min=-30
    onlyTwist=False
    parent=thigh_r_null
    primAxis=1
    twistAxis=(X=-1,Y=0,Z=0)
   )
   > RigVMVariableNode(Value=@"Delta Rotation_1_1_3".ouputWeight) {variable=legDeltaRot_SwingYIn_r}
   Remap = &RigVMFunction_MathDoubleRemap(
    SourceMaximum=0.2
    SourceMinimum=0
    TargetMaximum=1
    TargetMinimum=0
    bClamp=False
    Value=RigVMVariableNode() {variable=legDeltaRot_SwingYIn_r}.Value
   )
   > &RigUnit_ModifyTransforms(
    ItemToModify="((Item=(Type=Bone,Name=\"area_belt_fw\"),Transform=(Rotation=(X=0,Y=0,Z=0,W=1),Translation=(X=5.5045,Y=1.7836,Z=6.5981),Scale3D=(X=1,Y=1,Z=1))),(Item=(Type=Bone,Name=\"base_grenade_A\"),Transform=(Rotation=(X=-0.1349,Y=0.1236,Z=0.06,W=0.9813),Translation=(X=1, ...(truncated)"
    WeightMinimum=0
    Weight=&RigVMFunction_MathDoubleMul(B=@Remap.Result, A=RigVMVariableNode() {variable=legDeltaRot_SwingZup_r}.Value).Result
   )
   Delta Rotation_2 = RigVMFunctionReferenceNode(
    ParentType=Null
    RotationOrder=XYZ
    child=thigh_l
    debugRotation=False
    max=-90
    onlyTwist=False
    parent=thigh_l_null
    primAxis=2
    twistAxis=(X=1,Y=0,Z=0)
    ExecuteContext=@RigVMFunction_Sequence_1.B
   )
   > &RigUnit_ModifyTransforms(
    ItemToModify="((Item=(Type=Bone,Name=\"base_belt_fw_l\"),Transform=(Rotation=(X=0.1046,Y=0.0402,Z=0,W=0.9937),Translation=(X=2.3752,Y=0.9364,Z=10),Scale3D=(X=1,Y=1,Z=1))),(Item=(Type=Bone,Name=\"base_belt_fw_c\"),Transform=(Rotation=(X=0,Y=0,Z=0,W=1),Translation=(X=4.3017,Y ...(truncated)"
    WeightMinimum=0
    Weight=@"Delta Rotation_2".ouputWeight
   )
   > RigVMVariableNode(Value=@"Delta Rotation_2".ouputWeight) {variable=legDeltaRot_SwingZup_l}
   > Delta Rotation_2_1 = RigVMFunctionReferenceNode(
    ParentType=Null
    RotationOrder=XYZ
    child=thigh_l
    debugRotation=False
    max=-100
    min=-70
    onlyTwist=False
    parent=thigh_l_null
    primAxis=2
    twistAxis=(X=1,Y=0,Z=0)
   )
   > &RigUnit_ModifyTransforms(
    ItemToModify="((Item=(Type=Bone,Name=\"base_belt_fw_l\"),Transform=(Rotation=(X=-0.2292,Y=0,Z=0,W=0.9734),Translation=(X=2.2002,Y=1.0583,Z=0),Scale3D=(X=1,Y=1,Z=1))),(Item=(Type=Bone,Name=\"base_belt_sd_l\"),Transform=(Rotation=(X=0,Y=0,Z=0,W=1),Translation=(X=0.9236,Y=-0.2 ...(truncated)"
    WeightMinimum=0
    Weight=@"Delta Rotation_2_1".ouputWeight
   )
   > Delta Rotation_1_1_2 = RigVMFunctionReferenceNode(
    ParentType=Null
    RotationOrder=XYZ
    child=thigh_l
    debugRotation=False
    max=-90
    onlyTwist=False
    parent=thigh_l_null
    primAxis=1
    twistAxis=(X=1,Y=0,Z=0)
   )
   > &RigUnit_ModifyTransforms(
    ItemToModify="((Item=(Type=Bone,Name=\"base_bag_A\"),Transform=(Rotation=(X=0.0302,Y=-0.7338,Z=0.001,W=0.6787),Translation=(X=-0.6513,Y=-0.214,Z=-1.6927),Scale3D=(X=1,Y=1,Z=1))),(Item=(Type=Bone,Name=\"base_bag_B\"),Transform=(Rotation=(X=0.1316,Y=-0.4841,Z=-0.0617,W=0.8629 ...(truncated)"
    WeightMinimum=0
    Weight=@"Delta Rotation_1_1_2".ouputWeight
   )
   n1 = RigVMVariableNode() {variable=DBNO}
   Multiply = RigVMAggregateNode(B=RigVMVariableNode() {variable=legDeltaRot_SwingZup_l}.Value, C=@n1.Value, A=@"Delta Rotation_1_1_2".ouputWeight)
   > &RigUnit_ModifyTransforms(
    ItemToModify="((Item=(Type=Bone,Name=\"base_bag_area\"),Transform=(Rotation=(X=0,Y=0,Z=-0.3663,W=0.9305),Translation=(X=3.3892,Y=-1.617,Z=0),Scale3D=(X=1,Y=1,Z=1))),(Item=(Type=Bone,Name=\"dyn_bag_A\"),Transform=(Rotation=(X=-0.0335,Y=0.6019,Z=-0.133,W=0.7867),Translation=( ...(truncated)"
    WeightMinimum=0
    Weight=@Multiply.Result
   )
   > &RigUnit_ModifyTransforms(
    ItemToModify="((Item=(Type=Bone,Name=\"base_grenade_A\"),Transform=(Rotation=(X=0,Y=0,Z=0.1406,W=0.9901),Translation=(X=1.5801,Y=-1.1939,Z=0),Scale3D=(X=1,Y=1,Z=1))),(Item=(Type=Bone,Name=\"base_grenades_area\"),Transform=(Rotation=(X=0,Y=0,Z=0,W=1),Translation=(X=0,Y=0,Z=0 ...(truncated)"
    WeightMinimum=0
    Weight=@n1.Value
   )
   n2 = RigVMVariableNode() {variable=Crouching}
   &RigUnit_ModifyTransforms(
    ItemToModify="((Item=(Type=Bone,Name=\"base_grenade_A\"),Transform=(Rotation=(X=0,Y=0,Z=0,W=1),Translation=(X=0,Y=0,Z=0),Scale3D=(X=1,Y=1,Z=1))),(Item=(Type=Bone,Name=\"base_grenade_B\"),Transform=(Rotation=(X=0,Y=0,Z=0,W=1),Translation=(X=-1.8739,Y=-0.182,Z=0),Scale3D=(X=1 ...(truncated)"
    WeightMinimum=0
    ExecutePin=@RigVMFunction_Sequence_1.C
    Weight=@n2.Value
   )
   > &RigUnit_ModifyTransforms(
    ItemToModify=((Item=(Type=Bone,Name="base_belt_fw_l"),Transform=(Rotation=(X=0,Y=0,Z=0,W=1),Translation=(X=0,Y=0,Z=-1.5),Scale3D=(X=1,Y=1,Z=1))))
    WeightMinimum=0
    Weight=@n2.Value
   )
   > &RigUnit_ModifyTransforms(
    ItemToModify="((Item=(Type=Bone,Name=\"base_bag_area\"),Transform=(Rotation=(X=0.3378,Y=0,Z=0,W=0.9412),Translation=(X=-0.9634,Y=-1.232,Z=1.6744),Scale3D=(X=1,Y=1,Z=1))),(Item=(Type=Bone,Name=\"base_belt_sd_l\"),Transform=(Rotation=(X=0.1866,Y=0,Z=0,W=0.9824),Translation=(X ...(truncated)"
    WeightMinimum=0
    Weight=RigVMVariableNode() {variable=DBNO}.Value
   )
   > Delta Rotation_3 = RigVMFunctionReferenceNode(
    ParentType=Null
    RotationOrder=ZXY
    child=thigh_r
    debugRotation=False
    max=-135
    min=-85
    onlyTwist=False
    parent=thigh_r_null
    primAxis=2
    twistAxis=(X=1,Y=0,Z=0)
   )
   > &RigUnit_ModifyTransforms(
    ItemToModify="((Item=(Type=Bone,Name=\"base_grenades_area\"),Transform=(Rotation=(X=-0.105,Y=0.007,Z=0.0056,W=0.9944),Translation=(X=4.9714,Y=-1.3938,Z=7.0632),Scale3D=(X=1,Y=1,Z=1))),(Item=(Type=Bone,Name=\"base_grenade_A\"),Transform=(Rotation=(X=-0.097,Y=-0.1618,Z=-0.016 ...(truncated)"
    WeightMinimum=0
    Weight=&RigVMFunction_MathDoubleMul(A=@"Delta Rotation_3".ouputWeight, B=RigVMVariableNode() {variable=DBNO}.Value).Result
   )
   > &RigUnit_ModifyTransforms(
    ItemToModify="((Item=(Type=Bone,Name=\"base_grenades_area\"),Transform=(Rotation=(X=0,Y=0,Z=0,W=1),Translation=(X=0,Y=0,Z=0),Scale3D=(X=1,Y=1,Z=1))),(Item=(Type=Bone,Name=\"base_grenade_A\"),Transform=(Rotation=(X=0.0239,Y=-0.1253,Z=0.2248,W=0.966),Translation=(X=0.3462,Y=0 ...(truncated)"
    WeightMinimum=0
    Weight=&RigVMFunction_MathDoubleMul(A=@"Delta Rotation_3".ouputWeight, B=RigVMVariableNode() {variable=ForwardDBNO}.Value).Result
   )
   RigVMFunctionReturnNode(ExecuteContext=@RigVMFunction_Sequence_1.D)
   Delta Rotation_1_1_1 = RigVMFunctionReferenceNode(
    ExecuteContext=()
    ParentType=Null
    RotationOrder=XYZ
    child=thigh_r
    debugRotation=False
    max=-45
    onlyTwist=true
    parent=thigh_r_null
    primAxis=0
    twistAxis=(X=1,Y=0,Z=0)
   )
   &RigUnit_ModifyTransforms(
    ExecutePin=()
    ItemToModify=((Item=(Type=Bone,Name="base_belt_bk"),Transform=(Rotation=(X=-0.0009,Y=-0.004,Z=0.0133,W=0.9999),Translation=(X=5.4544,Y=0.2721,Z=0.6047),Scale3D=(X=1,Y=1,Z=1))))
    WeightMinimum=0
    Weight=@"Delta Rotation_1_1_1".ouputWeight
   )
   > RigVMVariableNode(Value=@"Delta Rotation_1_1_1".ouputWeight) {variable=legDeltaRot_SwingXout_r}
   Clamp = &RigVMFunction_MathDoubleClamp(Maximum=0.6, Value=RigVMVariableNode() {variable=legDeltaRot_SwingZup_r}.Value)
   > &RigUnit_ModifyTransforms(
    ItemToModify=((Item=(Type=Bone,Name="base_belt_bk"),Transform=(Rotation=(X=0.061,Y=-0.041,Z=0.0731,W=0.9946),Translation=(X=-12.7295,Y=-5.2302,Z=-13.6612),Scale3D=(X=1,Y=1,Z=1))))
    WeightMinimum=0
    Weight=&RigVMFunction_MathDoubleMul(B=@Clamp.Result, A=RigVMVariableNode() {variable=legDeltaRot_SwingXout_r}.Value).Result
   )
   &RigUnit_ModifyTransforms(
    ExecutePin=()
    ItemToModify=((Item=(Type=Bone,Name="base_belt_area"),Transform=(Rotation=(X=0.0585,Y=-0.1262,Z=0.1438,W=0.9798),Translation=(X=0,Y=0,Z=0),Scale3D=(X=1,Y=1,Z=1))))
    WeightMinimum=0
    Weight=&RigVMFunction_MathDoubleMul(B=RigVMVariableNode() {variable=legDeltaRot_SwingYOut_r}.Value, A=RigVMVariableNode() {variable=legDeltaRot_SwingZup_r}.Value).Result
   )
   &RigVMFunction_MathDoubleMul(
    B=&RigVMFunction_MathDoubleRemap(
     SourceMaximum=0.2
     SourceMinimum=0
     TargetMaximum=1
     TargetMinimum=0
     bClamp=False
     Value=RigVMVariableNode() {variable=legDeltaRot_SwingYIn_r}.Value
    ).Result
    A=RigVMVariableNode() {variable=legDeltaRot_SwingZup_r}.Value
   )
   RigVMVariableNode(ExecuteContext=(), Value=@"Delta Rotation_1_1".ouputWeight) {variable=legDeltaRot_SwingYOut_r}
  ]}
 New Function_ContainedGraph {nodes [
   n1 = RigVMFunctionEntryNode()
   RigVMFunction_Sequence = &RigVMFunction_Sequence(ExecuteContext=@n1.ExecuteContext)
   n2 = &RigUnit_SphericalPoseReader(
    ActiveRegionScaleFactors=(PositiveWidth=0,NegativeWidth=1,PositiveHeight=1,NegativeHeight=1)
    ActiveRegionSize=0.1
    Debug=(bDrawDebug=false,bDraw2D=false,DebugScale=25,DebugSegments=20,DebugThickness=0.25)
    DriverAxis=(X=-1,Y=0,Z=0)
    DriverItem=(Type=Bone,Name="calf_l")
    FalloffRegionScaleFactors=(PositiveWidth=0,NegativeWidth=1,PositiveHeight=1,NegativeHeight=1)
    FalloffSize=0.45
    OptionalParentItem=(Type=Bone,Name="thigh_l")
    RotationOffset=(X=-90,Y=0,Z=65)
    ExecutePin=@RigVMFunction_Sequence.A
   )
   > &RigUnit_ModifyTransforms(
    ItemToModify=((Item=(Type=Bone,Name="calfcounter_l"),Transform=(Rotation=(X=0,Y=0,Z=-0.0872,W=0.9962),Translation=(X=0,Y=0,Z=0),Scale3D=(X=1,Y=1,Z=1))))
    WeightMinimum=0
    Weight=@n2.OutputParam
   )
   > &RigUnit_ModifyTransforms(
    ItemToModify=((Item=(Type=Bone,Name="thighcounter_l"),Transform=(Rotation=(X=0,Y=0,Z=-0.0872,W=0.9962),Translation=(X=0,Y=0,Z=0),Scale3D=(X=1,Y=1,Z=1))))
    WeightMinimum=0
    Weight=@n2.OutputParam
   )
   > RigVMFunctionReturnNode()
   n3 = &RigUnit_SphericalPoseReader(
    ActiveRegionScaleFactors=(PositiveWidth=0,NegativeWidth=1,PositiveHeight=1,NegativeHeight=1)
    ActiveRegionSize=0.1
    Debug=(bDrawDebug=false,bDraw2D=false,DebugScale=25,DebugSegments=20,DebugThickness=0.25)
    DriverAxis=(X=1,Y=0,Z=0)
    DriverItem=(Type=Bone,Name="calf_r")
    FalloffRegionScaleFactors=(PositiveWidth=0,NegativeWidth=1,PositiveHeight=1,NegativeHeight=1)
    FalloffSize=0.45
    OptionalParentItem=(Type=Bone,Name="thigh_r")
    RotationOffset=(X=-90,Y=0,Z=-115)
    ExecutePin=@RigVMFunction_Sequence.B
   )
   > &RigUnit_ModifyTransforms(
    ItemToModify=((Item=(Type=Bone,Name="calfcounter_r"),Transform=(Rotation=(X=0,Y=0,Z=-0.0872,W=0.9962),Translation=(X=0,Y=0,Z=0),Scale3D=(X=1,Y=1,Z=1))))
    WeightMinimum=0
    Weight=@n3.OutputParam
   )
   > &RigUnit_ModifyTransforms(
    ItemToModify=((Item=(Type=Bone,Name="thighcounter_r"),Transform=(Rotation=(X=0,Y=0,Z=-0.0872,W=0.9962),Translation=(X=0,Y=0,Z=0),Scale3D=(X=1,Y=1,Z=1))))
    WeightMinimum=0
    Weight=@n3.OutputParam
   )
  ]}
 RigVMFunctionLibrary {nodes [
   BeltFix = RigVMCollapseNode(ExecuteContext=())
   Calf_FNC = RigVMCollapseNode(ExecuteContext=())
   Constrains = RigVMCollapseNode(ExecuteContext=())
   Controls = RigVMCollapseNode(DebugControls=False, ExecuteContext=())
   Delta Rotation = RigVMCollapseNode(ExecuteContext=(), debugRotation=False, onlyTwist=False, twistAxis=(X=0,Y=0,Z=0))
  ]}
 RigVMFunction_Sequence_1_ContainedGraph {nodes [
   n1 = RigVMFunctionEntryNode()
   RigVMFunction_Sequence = &RigVMFunction_Sequence(ExecuteContext=@n1.ExecuteContext)
   n2 = &RigVMFunction_Sequence(ExecuteContext=@RigVMFunction_Sequence.B)
   n3 = &RigVMFunction_Sequence(ExecuteContext=@n2.B)
   RigVMFunctionReturnNode(A=@RigVMFunction_Sequence.A, B=@n2.A, C=@n3.A, D=@n3.B)
  ]}
 RigVMModel {nodes [
   RigUnit_BeginExecution = &RigUnit_BeginExecution()
   Controls = RigVMFunctionReferenceNode(DebugControls=false, ExecuteContext=@RigUnit_BeginExecution.ExecutePin)
   > FN_DeformRig_FNC = RigVMFunctionReferenceNode(
    DeformRig_ Curve="(deformRig_l_upperarm_up_alpha_72_#1=False,deformRig_r_upperarm_up_alpha_73_#2=False,deformRig_l_upperarm_fr_alpha_74_#3=False,deformRig_r_upperarm_fr_alpha_75_CB51B8 ...(truncated)"
    DeformRig_Bool="(upperarm_1_#4=True,elbow_3_#5=True,wrist_7_#6=True,groin_8_#7=True,glute_17_#8=true,knee_9_40C408324CA261 ...(truncated)"
    DeformRig_Debug=(upperarm_debug_18_#4=False,wrist_debug_20_#6=False,lat_debug_26_#9=False,groinGlute_debug_27_#7=False)
    DeformRig_Value="(upperarm_up_18_#10=(Rotation=(X=0,Y=0,Z=0,W=1),Translation=(X=0,Y=0,Z=6),Scale3D=(X=0.7,Y=1,Z=1)),upperarm_fwd_fr_19_#11=(Rotation=(X=0,Y=0,Z=0,W=1),Translation=(X=0,Y=10,Z=0),Scale3D=(X=0.5,Y=1,Z= ...(truncated)"
   )
   n1 = RigVMVariableNode() {variable=Parachute}
   > &RigUnit_ModifyTransforms(
    ItemToModify=((Item=(Type=Bone,Name="deform_upperarm_up_l"),Transform=(Rotation=(X=0,Y=0,Z=0,W=1),Translation=(X=0,Y=0,Z=-5),Scale3D=(X=1,Y=1,Z=1))))
    WeightMinimum=0
    Weight=@n1.Value
   )
   > &RigUnit_ModifyTransforms(
    ItemToModify=((Item=(Type=Bone,Name="deform_upperarm_up_r"),Transform=(Rotation=(X=0,Y=0,Z=0,W=1),Translation=(X=0,Y=0,Z=5),Scale3D=(X=1,Y=1,Z=1))))
    WeightMinimum=0
    Weight=@n1.Value
   )
   > BeltFix = RigVMFunctionReferenceNode()
   > Calf_FNC = RigVMFunctionReferenceNode()
   > n2 = &RigUnit_SphericalPoseReader(
    ActiveRegionScaleFactors=(PositiveWidth=1,NegativeWidth=1,PositiveHeight=1,NegativeHeight=1)
    ActiveRegionSize=0.12
    Debug=(bDrawDebug=false,bDraw2D=false,DebugScale=25,DebugSegments=20,DebugThickness=0.25)
    DriverAxis=(X=1,Y=0,Z=0)
    DriverItem=(Type=Bone,Name="thumb_03_l")
    FalloffRegionScaleFactors=(PositiveWidth=1,NegativeWidth=1,PositiveHeight=1,NegativeHeight=1)
    FalloffSize=0.2
    OptionalParentItem=(Type=Bone,Name="lowerarm_l")
    RotationOffset=(X=-30,Y=30,Z=20)
   )
   > &RigUnit_ModifyTransforms(
    ItemToModify="((Item=(Type=Bone,Name=\"lowerarm_twist_01_l\"),Transform=(Rotation=(X=0,Y=0,Z=0,W=1),Translation=(X=0,Y=0,Z=0),Scale3D=(X=1,Y=1,Z=2))),(Item=(Type=Bone,Name=\"lowerarm_twist_02_l\"),Transform=(Rotation=(X=0,Y=0,Z=0,W=1),Translation=(X=0,Y=0,Z=0),Scale3D=(X=1, ...(truncated)"
    WeightMinimum=0
    Weight=@n2.OutputParam
   )
   Constrains = RigVMFunctionReferenceNode(ExecuteContext=())
  ]}
]
hierarchy [
 root {loc=<0 0 0>, rot=<0 0 0 1>}
 pelvis {parent=root, loc=<0 -0.7071 93.1191>, rot=<0.0127 0.707 -0.0127 -0.707>}
 spine_01 {parent=pelvis, loc=<6.7494 -0.6388 0>, rot=<0 0 0.034 -0.9994>}
 spine_02 {parent=spine_01, loc=<7.252 -0.035 0>, rot=<0 0 0.0238 -0.9997>}
 spine_03 {parent=spine_02, loc=<6.0106 0 0>, rot=<0 0 0.0149 -0.9999>}
 spine_04 {parent=spine_03, loc=<8.1023 0 0>, rot=<0 0 -0.0132 -0.9999>}
 spine_05 {parent=spine_04, loc=<10.0142 0 0>, rot=<0 0 -0.0015 -1>}
 attach_backpack {parent=spine_05, loc=<-0.123 -13.473 0>, rot=<-0.0093 -0.707 0.0093 -0.707>}
 neck_01 {parent=spine_05, loc=<11.3138 0.1529 0>, rot=<0 0 -0.2003 -0.9797>}
 neck_02 {parent=neck_01, loc=<4.5627 0 0>, rot=<0 0 0.0504 -0.9987>}
 head {parent=neck_02, loc=<5.3914 0 0>, rot=<0 0 0.115 -0.9934>}
 base_hair_sd_l_A {parent=head, loc=<9.4611 8.0564 -4.1544>, rot=<0.0245 -0.9223 -0.36 -0.138>}
 dyn_hair_sd_l_A_01 {parent=base_hair_sd_l_A, loc=<0 0 0>, rot=<0 0 0 1>}
 dyn_hair_sd_l_A_02 {parent=dyn_hair_sd_l_A_01, loc=<2.9791 0 0>, rot=<0 0 0 1>}
 base_hair_sd_l_B {parent=head, loc=<7.9845 6.9753 -5.5874>, rot=<0.1833 -0.8893 0.2832 -0.3087>}
 dyn_hair_sd_l_B_01 {parent=base_hair_sd_l_B, loc=<0 0 0>, rot=<0 0 0 1>}
 dyn_hair_sd_l_B_02 {parent=dyn_hair_sd_l_B_01, loc=<2.7443 0 0>, rot=<0 0 0 1>}
 base_hair_sd_l_C {parent=head, loc=<6.1888 4.2618 -6.6149>, rot=<0.2579 -0.7626 0.3809 -0.4547>}
 dyn_hair_sd_l_C_01 {parent=base_hair_sd_l_C, loc=<0 0 0>, rot=<0 0 0 1>}
 dyn_hair_sd_l_C_02 {parent=dyn_hair_sd_l_C_01, loc=<2.7443 0 0>, rot=<0 0 0 1>}
 base_hair_fw_area {parent=head, loc=<12.591 9.0729 0.8106>, rot=<0.4979 0.5021 0.5021 0.4979>}
 dyn_hair_fw_area {parent=base_hair_fw_area, loc=<0 0 0>, rot=<0 0 0 1>}
 base_fw_C {parent=dyn_hair_fw_area, loc=<0.0948 3.6478 -1.4544>, rot=<0.0475 -0.1297 -0.7534 -0.6428>}
 dyn_hair_fw_C_01 {parent=base_fw_C, loc=<0 0 0>, rot=<0 0 0 1>}
 dyn_hair_fw_C_02 {parent=dyn_hair_fw_C_01, loc=<2.0457 0 0>, rot=<0 0 0 1>}
 base_hair_fw_B {parent=dyn_hair_fw_area, loc=<1.4776 3.7495 -3.48>, rot=<0.0581 -0.2771 -0.6506 -0.7047>}
 dyn_hair_fw_B_01 {parent=base_hair_fw_B, loc=<0 0 0>, rot=<0 0 0 1>}
 dyn_hair_fw_B_02 {parent=dyn_hair_fw_B_01, loc=<2 0 0>, rot=<0 0.0032 0 -1>}
 dyn_hair_fw_B_03 {parent=dyn_hair_fw_B_01, loc=<1.6786 0 -1.3506>, rot=<-0.1945 -0.4119 0.0904 -0.8856>}
 base_fw_A {parent=dyn_hair_fw_area, loc=<1.188 -0.8009 -1.0345>, rot=<-0.0661 -0.2992 -0.2981 -0.904>}
 base_hair_sd_r_A {parent=head, loc=<5.5029 5.324 6.6909>, rot=<0.4723 -0.4909 0.6277 -0.3768>}
 dyn_hair_sd_r_A_01 {parent=base_hair_sd_r_A, loc=<0 0 0>, rot=<0 0 0 1>}
 dyn_hair_sd_r_A_02 {parent=dyn_hair_sd_r_A_01, loc=<3.3117 0 0>, rot=<0 0 0 1>}
 base_hat_area {parent=head, loc=<10.1374 1.2385 0>, rot=<0.5021 -0.4979 0.4979 -0.5021>}
 dyn_hat_area {parent=base_hat_area, loc=<0 0 0>, rot=<0 0 0 1>}
 base_hat {parent=dyn_hat_area, loc=<7.533 0 -3.1348>, rot=<0 0 0 1>}
 dyn_hat {parent=base_hat, loc=<0 0 0>, rot=<0 0 0 1>}
 FACIAL_C_FacialRoot {parent=head, loc=<2.1524 -0.746 -0.0002>, rot=<0.4879 -0.5118 -0.4879 -0.5118>}
 FACIAL_C_UnderChin {parent=FACIAL_C_FacialRoot, loc=<0.001 8.3549 5.431>, rot=<0.4957 0.0001 0 -0.8685>}
 FACIAL_L_UnderChin {parent=FACIAL_C_FacialRoot, loc=<3.2131 7.1463 4.0388>, rot=<0.3985 -0.3552 0.0806 -0.8417>}
 FACIAL_R_UnderChin {parent=FACIAL_C_FacialRoot, loc=<-3.2219 7.1519 4.0353>, rot=<0.3985 0.3553 -0.0806 -0.8417>}
 FACIAL_C_Jaw {parent=FACIAL_C_FacialRoot, loc=<0.0031 1.269 2.7817>, rot=<0.3376 0.0001 0 -0.9413>}
 FACIAL_C_Jawline {parent=FACIAL_C_Jaw, loc=<-0.0116 2.5372 8.034>, rot=<0.2586 0 0 -0.966>}
 neck_correct {parent=neck_02, loc=<1.1447 1.188 0>, rot=<-0.5523 -0.4415 -0.4415 -0.5523>}
 clavicle_l {parent=spine_05, loc=<4.3575 -0.1466 -2.3223>, rot=<0.0465 -0.7289 -0.009 -0.6829>}
 upperarm_l {parent=clavicle_l, loc=<15.5111 0 0>, rot=<0.0226 -0.3234 -0.0088 -0.9459>}
 lowerarm_l {parent=upperarm_l, loc=<26.6514 0 0>, rot=<0.0391 -0.0099 -0.2453 -0.9686>}
 hand_l {parent=lowerarm_l, loc=<22.3491 0 0>, rot=<-0.7081 -0.0507 0.0975 -0.6975>}
 thumb_01_l {parent=hand_l, loc=<1.3118 -0.3191 -2.459>, rot=<0.6546 -0.4042 0.0664 -0.6354>}
 thumb_02_l {parent=thumb_01_l, loc=<5.9336 -0.0448 0.0401>, rot=<-0.0037 0.0213 0.298 -0.9543>}
 thumb_03_l {parent=thumb_02_l, loc=<2.2223 0.3682 -0.0907>, rot=<0.0118 -0.013 0.0066 -0.9998>}
 index_metacarpal_l {parent=hand_l, loc=<1.6535 0.0183 -2.96>, rot=<0.0158 -0.0124 -0.0173 -0.9996>}
 index_01_l {parent=index_metacarpal_l, loc=<8.3008 0.0357 -0.0524>, rot=<-0.0044 0.0485 0.0924 -0.9945>}
 index_02_l {parent=index_01_l, loc=<3.3587 -0.3004 0.0452>, rot=<-0.002 0.0027 0.3225 -0.9466>}
 index_03_l {parent=index_02_l, loc=<2.5578 -0.123 -0.0252>, rot=<-0.002 0.0067 0.2558 -0.9667>}
 middle_metacarpal_l {parent=hand_l, loc=<2.1856 -0.1725 -0.5984>, rot=<0.006 0.0168 -0.0329 -0.9993>}
 middle_01_l {parent=middle_metacarpal_l, loc=<7.3275 -0.2447 0.0129>, rot=<-0.0208 0.0042 0.0858 -0.9961>}
 middle_02_l {parent=middle_01_l, loc=<4.0123 0.1401 0.1175>, rot=<-0.0218 0.0258 0.3512 -0.9357>}
 middle_03_l {parent=middle_02_l, loc=<2.6077 0.0534 0.1373>, rot=<0.0082 0.0188 0.1708 -0.9851>}
 ring_metacarpal_l {parent=hand_l, loc=<2.2074 -0.3625 1.3638>, rot=<-0.0284 0.0722 -0.0458 -0.9959>}
 ring_01_l {parent=ring_metacarpal_l, loc=<7.0295 -0.6197 -0.1446>, rot=<-0.0489 -0.0252 0.1892 -0.9804>}
 ring_02_l {parent=ring_01_l, loc=<3.6174 0.4647 -0.1027>, rot=<0 0 0.2806 -0.9598>}
 ring_03_l {parent=ring_02_l, loc=<2.4152 0.114 0.0199>, rot=<0 0 0.1189 -0.9929>}
 pinky_metacarpal_l {parent=hand_l, loc=<1.7616 -0.9753 3.3814>, rot=<-0.0397 0.0814 -0.069 -0.9935>}
 pinky_01_l {parent=pinky_metacarpal_l, loc=<7.1709 -0.4831 -0.1094>, rot=<-0.0654 -0.0035 0.2242 -0.9723>}
 pinky_02_l {parent=pinky_01_l, loc=<2.8274 0.0893 -0.1043>, rot=<0.0037 0.0011 0.2854 -0.9584>}
 pinky_03_l {parent=pinky_02_l, loc=<2.2046 0.0307 -0.0078>, rot=<-0.0002 0.0098 0.1445 -0.9895>}
 deform_wrist_dn_base_l {parent=hand_l, loc=<0 -2.2342 0>, rot=<0 0 0 -1>}
 deform_wrist_dn_l {parent=deform_wrist_dn_base_l, loc=<0 0 0>, rot=<0 0 0 1>}
 deform_wrist_up_base_l {parent=hand_l, loc=<0 2.2682 0>, rot=<0 0 0 -1>}
 deform_wrist_up_l {parent=deform_wrist_up_base_l, loc=<0 0 0>, rot=<0 0 0 1>}
 hand_attach_l {parent=hand_l, loc=<8.5 -3 0>, rot=<0.7071 0.7071 0 0>}
 weapon_l {parent=hand_l, loc=<-0.1974 -3.4111 0.3811>, rot=<0 0 0 1>}
 hand_l_null {parent=hand_l, type=Null, loc=<0 0 0>, rot=<0 0 0 1>}
 lowerarm_twist_02_l {parent=lowerarm_l, loc=<9.6199 0 0>, rot=<0 0 0 -1>}
 lowerarm_twist_01_l {parent=lowerarm_l, loc=<17.8549 0 0>, rot=<0 0 0 -1>}
 deform_elbow_in_base_l {parent=lowerarm_l, loc=<0.2112 5.259 0.8274>, rot=<0 0 0 -1>}
 deform_elbow_in_l {parent=deform_elbow_in_base_l, loc=<0 0 0>, rot=<0 0 0 1>}
 deform_elbow_out_base_l {parent=lowerarm_l, loc=<0 -4 0>, rot=<0 0 0 -1>}
 deform_elbow_out_l {parent=deform_elbow_out_base_l, loc=<0 0 0>, rot=<0 0 0 1>}
 upperarm_twist_01_l {parent=upperarm_l, loc=<0.0041 0 0>, rot=<0 0 0 -1>}
 deform_upperarm_bk_base_l {parent=upperarm_twist_01_l, loc=<1.0813 -7.2193 -1.9453>, rot=<0 0 0 1>}
 deform_upperarm_bk_l {parent=deform_upperarm_bk_base_l, loc=<0 0 0>, rot=<0 0 0 1>}
 deform_upperarm_fr_base_l {parent=upperarm_twist_01_l, loc=<-0.3691 7.0506 -2.765>, rot=<0 0 0 1>}
 deform_upperarm_fr_l {parent=deform_upperarm_fr_base_l, loc=<0 0 0>, rot=<0 0 0 1>}
 deform_upperarm_up_base_l {parent=upperarm_twist_01_l, loc=<-3.2216 -1.1391 3.3311>, rot=<0 0 0 1>}
 deform_upperarm_up_l {parent=deform_upperarm_up_base_l, loc=<0 0 0>, rot=<0 0 0 1>}
 upperarm_twist_02_l {parent=upperarm_l, loc=<13.7941 0 0>, rot=<0 0 0 -1>}
 base_sleeves_fw_l {parent=upperarm_l, loc=<10.138 0.1565 3.2476>, rot=<-0.0082 0.0388 -0.0175 -0.9991>}
 dyn_sleeves_fw_l {parent=base_sleeves_fw_l, loc=<0 0 0>, rot=<0 0 0 1>}
 base_sleeves_dn_l {parent=upperarm_l, loc=<9.718 0.2734 -5.5204>, rot=<-0.0082 0.0388 -0.0175 -0.9991>}
 dyn_sleeves_dn_l {parent=base_sleeves_dn_l, loc=<0 0 0>, rot=<0 0 0 1>}
 attach_cape {parent=spine_05, loc=<9.017 -11.714 0>, rot=<0.0616 -0.7044 -0.0616 -0.7044>}
 clavicle_r {parent=spine_05, loc=<4.3575 -0.1466 2.3223>, rot=<0.7289 0.0465 -0.6829 0.009>}
 upperarm_r {parent=clavicle_r, loc=<-15.5111 0 0>, rot=<0.0226 -0.3234 -0.0088 -0.9459>}
 lowerarm_r {parent=upperarm_r, loc=<-26.6514 0 0>, rot=<0.0391 -0.0099 -0.2453 -0.9686>}
 hand_r {parent=lowerarm_r, loc=<-22.3491 0 0>, rot=<-0.7081 -0.0507 0.0975 -0.6975>}
 thumb_01_r {parent=hand_r, loc=<-1.3118 0.3191 2.459>, rot=<0.6546 -0.4042 0.0664 -0.6354>}
 thumb_02_r {parent=thumb_01_r, loc=<-5.9336 0.0448 -0.0401>, rot=<-0.0037 0.0213 0.298 -0.9543>}
 thumb_03_r {parent=thumb_02_r, loc=<-2.2223 -0.3682 0.0907>, rot=<0.0118 -0.013 0.0066 -0.9998>}
 index_metacarpal_r {parent=hand_r, loc=<-1.6535 -0.0183 2.96>, rot=<0.0158 -0.0124 -0.0173 -0.9996>}
 index_01_r {parent=index_metacarpal_r, loc=<-8.3008 -0.0357 0.0524>, rot=<-0.0044 0.0485 0.0924 -0.9945>}
 index_02_r {parent=index_01_r, loc=<-3.3587 0.3004 -0.0452>, rot=<-0.002 0.0027 0.3225 -0.9466>}
 index_03_r {parent=index_02_r, loc=<-2.5578 0.123 0.0252>, rot=<-0.002 0.0067 0.2558 -0.9667>}
 middle_metacarpal_r {parent=hand_r, loc=<-2.1856 0.1725 0.5984>, rot=<0.006 0.0168 -0.0329 -0.9993>}
 middle_01_r {parent=middle_metacarpal_r, loc=<-7.3275 0.2447 -0.0129>, rot=<-0.0208 0.0042 0.0858 -0.9961>}
 middle_02_r {parent=middle_01_r, loc=<-4.0123 -0.1401 -0.1175>, rot=<-0.0218 0.0258 0.3512 -0.9357>}
 middle_03_r {parent=middle_02_r, loc=<-2.6077 -0.0534 -0.1373>, rot=<0.0082 0.0188 0.1708 -0.9851>}
 ring_metacarpal_r {parent=hand_r, loc=<-2.2074 0.3625 -1.3638>, rot=<-0.0284 0.0722 -0.0458 -0.9959>}
 ring_01_r {parent=ring_metacarpal_r, loc=<-7.0295 0.6197 0.1446>, rot=<-0.0489 -0.0252 0.1892 -0.9804>}
 ring_02_r {parent=ring_01_r, loc=<-3.6174 -0.4647 0.1027>, rot=<0 0 0.2806 -0.9598>}
 ring_03_r {parent=ring_02_r, loc=<-2.4152 -0.114 -0.0199>, rot=<0 0 0.1189 -0.9929>}
 pinky_metacarpal_r {parent=hand_r, loc=<-1.7616 0.9753 -3.3814>, rot=<-0.0397 0.0814 -0.069 -0.9935>}
 pinky_01_r {parent=pinky_metacarpal_r, loc=<-7.1709 0.4831 0.1094>, rot=<-0.0654 -0.0035 0.2242 -0.9723>}
 pinky_02_r {parent=pinky_01_r, loc=<-2.8274 -0.0893 0.1043>, rot=<0.0037 0.0011 0.2854 -0.9584>}
 pinky_03_r {parent=pinky_02_r, loc=<-2.2046 -0.0307 0.0078>, rot=<-0.0002 0.0098 0.1445 -0.9895>}
 deform_wrist_dn_base_r {parent=hand_r, loc=<-0.0004 2.2345 -0.0001>, rot=<0 0 0 -1>}
 deform_wrist_dn_r {parent=deform_wrist_dn_base_r, loc=<0 0 0>, rot=<0 0 0 1>}
 deform_wrist_up_base_r {parent=hand_r, loc=<0.0003 -2.2684 0.0001>, rot=<0 0 0 -1>}
 deform_wrist_up_r {parent=deform_wrist_up_base_r, loc=<0 0 0>, rot=<0 0 0 1>}
 weapon_r {parent=hand_r, loc=<0.1974 3.4111 -0.3811>, rot=<0 0 0 1>}
 hand_attach_r {parent=hand_r, loc=<-8.5001 2.9999 0>, rot=<0 0 -0.7071 -0.7071>}
 lowerarm_twist_02_r {parent=lowerarm_r, loc=<-9.6199 0 0>, rot=<0 0 0 -1>}
 lowerarm_twist_01_r {parent=lowerarm_r, loc=<-17.8549 0 0>, rot=<0 0 0 -1>}
 deform_elbow_in_base_r {parent=lowerarm_r, loc=<-0.2111 -5.2591 -0.8275>, rot=<0 0 0 1>}
 deform_elbow_in_r {parent=deform_elbow_in_base_r, loc=<-0.0001 0 0.0002>, rot=<0 0 0 1>}
 deform_elbow_out_base_r {parent=lowerarm_r, loc=<0 4 -0.0001>, rot=<0 0 0 1>}
 deform_elbow_out_r {parent=deform_elbow_out_base_r, loc=<0 0 0>, rot=<0 0 0 1>}
 upperarm_twist_01_r {parent=upperarm_r, loc=<-0.0041 0 0>, rot=<0 0 0 -1>}
 deform_upperarm_bk_base_r {parent=upperarm_twist_01_r, loc=<-1.0814 7.2192 1.9455>, rot=<0 0 0 1>}
 deform_upperarm_bk_r {parent=deform_upperarm_bk_base_r, loc=<0 0 0>, rot=<0 0 0 1>}
 deform_upperarm_fr_base_r {parent=upperarm_twist_01_r, loc=<0.3689 -7.0506 2.7652>, rot=<0 0 0 1>}
 deform_upperarm_fr_r {parent=deform_upperarm_fr_base_r, loc=<0 0 0>, rot=<0 0 0 1>}
 deform_upperarm_up_base_r {parent=upperarm_twist_01_r, loc=<3.2214 1.1391 -3.3308>, rot=<0 0 0 1>}
 deform_upperarm_up_r {parent=deform_upperarm_up_base_r, loc=<0 0 0>, rot=<0 0 0 -1>}
 upperarm_twist_02_r {parent=upperarm_r, loc=<-13.7941 0 0>, rot=<0 0 0 -1>}
 base_sleeves_dn_r {parent=upperarm_r, loc=<-9.7177 -0.2734 5.5202>, rot=<-0.0082 0.0388 -0.0175 -0.9991>}
 dyn_sleeves_dn_r {parent=base_sleeves_dn_r, loc=<0 0 0>, rot=<0 0 0 1>}
 base_sleeves_fw_r {parent=upperarm_r, loc=<-10.138 -0.1565 -3.2476>, rot=<-0.0082 0.0388 -0.0175 -0.9991>}
 dyn_sleeves_fw_r {parent=base_sleeves_fw_r, loc=<0 0 0>, rot=<0 0 0 1>}
 deform_pec_aim_base_l {parent=spine_04, loc=<7.8728 12.6388 -2.4123>, rot=<0.0102 -0.4602 0.3979 -0.7936>}
 deform_pec_aim_l {parent=deform_pec_aim_base_l, loc=<0 0 0>, rot=<0 0 0 1>}
 deform_pec_base_l {parent=deform_pec_aim_l, loc=<10.766 0 0>, rot=<0 0 0 1>}
 deform_pec_l {parent=deform_pec_base_l, loc=<0.055 0 0>, rot=<0 0 0 1>}
 deform_pec_aim_base_r {parent=spine_04, loc=<7.8723 12.6388 2.4123>, rot=<-0.4602 -0.0102 0.7936 0.3979>}
 deform_pec_aim_r {parent=deform_pec_aim_base_r, loc=<0 0 0>, rot=<0 0 0 1>}
 deform_pec_base_r {parent=deform_pec_aim_r, loc=<-10.7662 -0.0006 -0.0007>, rot=<0 0 0 1>}
 deform_pec_r {parent=deform_pec_base_r, loc=<0 0 0>, rot=<0 0 0 1>}
 thigh_l {parent=pelvis, loc=<-2.5693 -0.0109 -7.7501>, rot=<0.0012 0.0572 0.0155 -0.9982>}
 calf_l {parent=thigh_l, loc=<-39.0484 0 0>, rot=<0.0006 -0.0161 -0.0391 -0.9991>}
 foot_l {parent=calf_l, loc=<-42.561 0.1077 0.4292>, rot=<0.0418 -0.0455 0.0343 -0.9975>}
 ball_l {parent=foot_l, loc=<-6.1599 13.8264 0.6806>, rot=<0.0001 -0.0001 -0.7063 -0.7079>}
 calf_twist_02_l {parent=calf_l, loc=<-18.8064 -0.0967 0.0397>, rot=<0.0237 0.0095 -0.0035 -0.9997>}
 calf_twist_01_l {parent=calf_l, loc=<-35.813 -0.0967 0.0397>, rot=<0.0237 0.0095 -0.0035 -0.9997>}
 deform_knee_in_base_l {parent=calf_l, loc=<1.298 -5.2957 -0.0496>, rot=<0 0 0 -1>}
 deform_knee_in_l {parent=deform_knee_in_base_l, loc=<0 0 0>, rot=<0 0 0 1>}
 deform_knee_out_base_l {parent=calf_l, loc=<0 4 0>, rot=<0 0 0 -1>}
 deform_knee_out_l {parent=deform_knee_out_base_l, loc=<0 0 0>, rot=<0 0 0 1>}
 calfcounter_l {parent=calf_l, loc=<0 0 0>, rot=<0 0 0 1>}
 thigh_twist_01_l {parent=thigh_l, loc=<-21.6303 0 0>, rot=<0.0013 0 0 -1>}
 thighcounter_l {parent=thigh_twist_01_l, loc=<21.6303 0 0>, rot=<-0.0013 0 0 -1>}
 deform_groin_base_l {parent=thigh_l, loc=<-2.1908 8.1142 -1.7442>, rot=<0 0 0 -1>}
 deform_groin_l {parent=deform_groin_base_l, loc=<0 0 0>, rot=<0 0 0 1>}
 thigh_r {parent=pelvis, loc=<-2.5693 -0.0109 7.7501>, rot=<0.0572 -0.0012 0.9982 0.0155>}
 calf_r {parent=thigh_r, loc=<39.0484 0 0>, rot=<0.0006 -0.0161 -0.0391 -0.9991>}
 foot_r {parent=calf_r, loc=<42.561 -0.1077 -0.4292>, rot=<0.0418 -0.0455 0.0343 -0.9975>}
 ball_r {parent=foot_r, loc=<6.1599 -13.8264 -0.6806>, rot=<0.0001 -0.0001 -0.7063 -0.7079>}
 calf_twist_02_r {parent=calf_r, loc=<18.8064 0.0967 -0.0397>, rot=<0.0237 0.0095 -0.0035 -0.9997>}
 calf_twist_01_r {parent=calf_r, loc=<35.813 0.0967 -0.0397>, rot=<0.0237 0.0095 -0.0035 -0.9997>}
 deform_knee_in_base_r {parent=calf_r, loc=<-1.298 5.2957 0.0496>, rot=<0 0 0 -1>}
 deform_knee_in_r {parent=deform_knee_in_base_r, loc=<0 0 0>, rot=<0 0 0 1>}
 deform_knee_out_base_r {parent=calf_r, loc=<0 -4 0>, rot=<0 0 0 -1>}
 deform_knee_out_r {parent=deform_knee_out_base_r, loc=<0 0 0>, rot=<0 0 0 1>}
 calfcounter_r {parent=calf_r, loc=<0 0 0>, rot=<0 0 0 1>}
 thigh_twist_01_r {parent=thigh_r, loc=<21.6303 0 0>, rot=<0.0013 0 0 -1>}
 thighcounter_r {parent=thigh_twist_01_r, loc=<-21.6303 0 0>, rot=<-0.0013 0 0 -1>}
 deform_groin_base_r {parent=thigh_r, loc=<2.1908 -8.1142 1.7442>, rot=<0 0 0 -1>}
 deform_groin_r {parent=deform_groin_base_r, loc=<0 0 0>, rot=<0 0 0 1>}
 base_bag_area {parent=pelvis, loc=<1.5602 2.1933 -11.101>, rot=<-0.0127 -0.707 0.0127 -0.707>}
 base_bag_A {parent=base_bag_area, loc=<2.0389 0.2692 -0.0028>, rot=<0 0 -0.0358 -0.9994>}
 dyn_bag_A {parent=base_bag_A, loc=<0 0 0>, rot=<0 0 0 1>}
 dyn_bag_A_loop {parent=dyn_bag_A, loc=<2.6687 4.0181 -2.5906>, rot=<0.9972 0.0416 0.0454 -0.0416>}
 base_bag_B {parent=base_bag_area, loc=<1.8627 -5.9524 -0.0028>, rot=<0 0 0.058 -0.9983>}
 dyn_bag_B {parent=base_bag_B, loc=<0 0 0>, rot=<0 0 0 1>}
 dyn_bag_B_loop {parent=dyn_bag_B, loc=<2.0591 -3.8157 -2.452>, rot=<0.9972 -0.0454 0.0416 -0.0416>}
 area_belt_fw {parent=pelvis, loc=<0.9294 3.0436 -3.3595>, rot=<-0.426 -0.5644 -0.4055 -0.5793>}
 base_belt_fw_c {parent=area_belt_fw, loc=<4.4344 1.7152 -0.9587>, rot=<0 0 -0.156 -0.9878>}
 dyn_belt_fw_c {parent=base_belt_fw_c, loc=<0 0 0>, rot=<0 0 0 1>}
 base_belt_fw_l {parent=area_belt_fw, loc=<5.3031 -3.5008 1.2097>, rot=<0 0 0.0267 -0.9996>}
 dyn_belt_fw_l {parent=base_belt_fw_l, loc=<0 0 0>, rot=<0 0 0 1>}
 thigh_l_null {parent=pelvis, type=Null, loc=<-2.5693 -0.0109 -7.7501>, rot=<0.0012 0.0572 0.0155 -0.9982>}
 thigh_r_null {parent=pelvis, type=Null, loc=<-2.5693 -0.0109 7.7501>, rot=<0.0572 -0.0012 0.9982 0.0155>}
 base_belt_area {parent=pelvis, loc=<1.4145 -3.8365 -8.2999>, rot=<0 0 0.0179 -0.9998>}
 dyn_belt_area {parent=base_belt_area, loc=<0 0 0>, rot=<0 0 0 1>}
 base_belt_bk {parent=dyn_belt_area, loc=<-5.0497 -8.607 8.1448>, rot=<0.5 -0.5 0.5 -0.5>}
 dyn_belt_bk {parent=base_belt_bk, loc=<0 0 0>, rot=<0 0 0 1>}
 base_grenades_area {parent=dyn_belt_area, loc=<-3.9041 1.8088 14.4134>, rot=<0.7071 0 0.7071 0>}
 base_grenade_A {parent=base_grenades_area, loc=<4.0536 -8.322 -1.6905>, rot=<0 0 0.5355 -0.8446>}
 dyn_grenade_A {parent=base_grenade_A, loc=<0 0 0>, rot=<0 0 0 1>}
 dyn_grenade_A_loop {parent=dyn_grenade_A, loc=<6.4301 -0.9079 4.1361>, rot=<0.9813 -0.0599 -0.1236 0.1349>}
 base_grenade_B {parent=base_grenades_area, loc=<9.3694 -2.6651 -4.8778>, rot=<0.1338 -0.0235 0.1714 -0.9758>}
 dyn_grenade_B {parent=base_grenade_B, loc=<0 0 0>, rot=<0 0 0 1>}
 dyn_grenade_B_loop {parent=dyn_grenade_B, loc=<7.2079 -0.8027 3.6918>, rot=<0.9763 0.0189 -0.2156 0>}
 base_grenade_C {parent=base_grenades_area, loc=<7.7904 4.3854 -7.3736>, rot=<0.0916 -0.0964 -0.393 -0.9099>}
 dyn_grenade_C {parent=base_grenade_C, loc=<0 0 0>, rot=<0 0 0 1>}
 dyn_grenade_C_loop {parent=dyn_grenade_C, loc=<7.2648 -1.3977 4.3466>, rot=<0.9782 -0.1083 -0.1753 -0.0278>}
 deform_glute_base_l {parent=pelvis, loc=<-8.8064 -8.8222 -8.4287>, rot=<0.0275 0.0501 -0.4472 -0.8926>}
 deform_glute_l {parent=deform_glute_base_l, loc=<0 0 0>, rot=<0 0 0 1>}
 deform_glute_base_r {parent=pelvis, loc=<-8.8062 -8.8222 8.4287>, rot=<0.0501 -0.0275 0.8926 -0.4472>}
 deform_glute_r {parent=deform_glute_base_r, loc=<0 0 0>, rot=<0 0 0 1>}
 base_belt_sd_l {parent=pelvis, loc=<4.0995 -0.8957 -12.1521>, rot=<-0.0127 -0.707 0.0127 -0.707>}
 dyn_belt_sd_l {parent=base_belt_sd_l, loc=<0 0 0>, rot=<0 0 0 1>}
 thigh_l_ctrl {controlType=EULER_TRANSFORM, type=Control, loc=<-104.166 1.3241 21.553>, rot=<0.8175 -0.0025 0.576 0.0019>}
 calf_l_ctrl {controlType=EULER_TRANSFORM, parent=thigh_l_ctrl, type=Control, loc=<0 0 0>, rot=<0 0 0 1>}
 thigh_r_ctrl {controlType=EULER_TRANSFORM, type=Control, loc=<0 0 0>, rot=<0 0 0 1>}
 calf_r_ctrl {controlType=EULER_TRANSFORM, parent=thigh_r_ctrl, type=Control, loc=<0 0 0>, rot=<0 0 0 1>}
 hand_l_ctrl {controlType=EULER_TRANSFORM, type=Control, loc=<0 0 0>, rot=<0 0 0 1>}
]
variables [
 Crouching {type=double, default=0}
 DBNO {type=double, default=0}
 ForwardDBNO {type=double, default=0}
 Parachute {type=double, default=0}
]
