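# Animation Blueprint for the VelvetDeskCam male cosmetic body, adding secondary motion and corrective posing on top of the parent mesh's pose.
# Copied body and head poses are blended per bone, then fed through a Control Rig driven by crouch, DBNO and parachute states.
# AnimDynamics then simulates sleeves, bags and grenades, followed by neck and jaw corrections and additive belt, bag and grenade offsets.
# The event graph computes gravity, state alphas and the head component reference.
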AnimBlueprint /BRCosmetics/Characters/Player/Male/Medium/Bodies/M_MED_VelvetDeskCam/Meshes/M_MED_VelvetDeskCam_AnimBP
target_skeleton=&/M_MED_VelvetDeskCam_Skeleton
graphs [
 AnimGraph {nodes [
   n1 = Use cached pose 'Dynamics Cache'()
   Blend Poses by bool(BlendPose_1=Component To Local(ComponentPose=Local To Component(LocalPose=@n1)), bActiveValue=NOT Boolean(A=Get bIsBaseballSliding())) {blendType=Linear}
   Component To Local()
   "Constraint - base_belt_fw_l (PPPP)"!(
    ComponentPose
    bAlphaBoolEnabled:bool=True
    ConstraintWeights_0=0.6163
    ConstraintWeights_1:float=0.2404
    ConstraintWeights_2:float=0.1402
    ConstraintWeights_3:float=0.0032
   ) -> (Pose) {
    alphaInputType=Bool
    alphaBoolBlend {blendOption=Linear}
    boneToModify {boneName=base_belt_fw_l}
    constraintSetup [
     {targetBone {boneName=pelvis}}
     {targetBone {boneName=thigh_twist_01_l}}
     {targetBone {boneName=dyn_belt_fw_c}}
     {targetBone {boneName=deform_groin_l}}
    ]
    constraintWeights [1,1,1,1]
   }
   "Transform (Modify) Bone - Bone: base_belt_bk"!(ComponentPose, bAlphaBoolEnabled:bool=True, Translation="-3.744,-5.082,-1.687", Rotation="-2.741,0.932,-6.160", Scale="1,1,1") -> (Pose) {
    alphaInputType=Bool
    rotationMode=BMM_Additive
    rotationSpace=BCS_BoneSpace
    translationMode=BMM_Additive
    translationSpace=BCS_BoneSpace
    alphaBoolBlend {blendInTime=1, blendOption=Linear, blendOutTime=0.2}
    boneToModify {boneName=base_belt_bk}
    rotation=<rot -2.741 0.932 -6.16>
    translation=<-3.744 -5.082 -1.687>
   }
   n2 = "Transform (Modify) Bone - Bone: base_belt_area"!(ComponentPose, bAlphaBoolEnabled:bool=True, Translation="-10.073,-1.025,-1.244", Rotation="-52.375,2.694,-5.224", Scale="1,1,1") -> (Pose) {
    alphaInputType=Bool
    rotationMode=BMM_Additive
    rotationSpace=BCS_BoneSpace
    translationMode=BMM_Additive
    translationSpace=BCS_BoneSpace
    alphaBoolBlend {blendInTime=1.3, blendOption=Linear, blendOutTime=0.2}
    boneToModify {boneName=base_belt_area}
    rotation=<rot -52.375 2.694 -5.224>
    translation=<-10.073 -1.025 -1.244>
   }
   n3 = "Transform (Modify) Bone - Bone: base_belt_bk"(ComponentPose=@n2, bAlphaBoolEnabled=True, Translation="-4.698,-8.172,-1.968", Rotation="-2.741,0.932,-6.160", Scale="1,1,1") {
    alphaInputType=Bool
    rotationMode=BMM_Additive
    rotationSpace=BCS_BoneSpace
    translationMode=BMM_Additive
    translationSpace=BCS_BoneSpace
    alphaBoolBlend {blendInTime=1, blendOption=Linear, blendOutTime=0.2}
    boneToModify {boneName=base_belt_bk}
    rotation=<rot -2.741 0.932 -6.16>
    translation=<-4.698 -8.172 -1.968>
   }
   "Transform (Modify) Bone - Bone: area_belt_fw"!(ComponentPose=@n3, bAlphaBoolEnabled:bool=True, Translation="1.485,2.024,-0.000", Rotation="0,0,0", Scale="1,1,1") -> (Pose) {
    alphaInputType=Bool
    rotationMode=BMM_Additive
    rotationSpace=BCS_BoneSpace
    translationMode=BMM_Additive
    translationSpace=BCS_BoneSpace
    alphaBoolBlend {blendInTime=0.5, blendOption=Linear, blendOutTime=0.2}
    boneToModify {boneName=area_belt_fw}
    translation=<1.4848 2.024 0>
   }
   n4 = "Transform (Modify) Bone - Bone: dyn_belt_area"!(
    ComponentPose=Local To Component(LocalPose=Use cached pose 'Dynamics Cache'())
    bAlphaBoolEnabled:bool=Get FrontEndIdle() -> (FrontEndIdle:bool)
    Translation="0.7,0.5,-1"
    Rotation="0,0,0"
    Scale="1,1,1"
   ) -> (Pose) {
    alphaInputType=Bool
    rotationMode=BMM_Additive
    rotationSpace=BCS_BoneSpace
    translationMode=BMM_Additive
    translationSpace=BCS_BoneSpace
    alphaBoolBlend {blendInTime=0.1, blendOption=Linear, blendOutTime=0.2}
    boneToModify {boneName=dyn_belt_area}
    translation=<0.7 0.5 -1>
   }
   n5 = "Transform (Modify) Bone - Bone: base_bag_A"!(ComponentPose=@n4, bAlphaBoolEnabled:bool=True, Translation="2.322,-0.89,-0.31", Rotation="-18.224,-20.865,-10.691", Scale="1,1,1") -> (Pose) {
    alphaInputType=Bool
    rotationMode=BMM_Additive
    rotationSpace=BCS_BoneSpace
    translationMode=BMM_Additive
    translationSpace=BCS_BoneSpace
    alphaBoolBlend {blendInTime=0.2, blendOption=Linear, blendOutTime=0.2}
    boneToModify {boneName=base_bag_A}
    rotation=<rot -18.224 -20.865 -10.6911>
    translation=<2.322 -0.89 -0.31>
   }
   n6 = "Transform (Modify) Bone - Bone: base_bag_B"!(ComponentPose=@n5, bAlphaBoolEnabled:bool=True, Translation="0.000,-1.506,-0.000", Rotation="-20.928,0.000,-0.000", Scale="1,1,1") -> (Pose) {
    alphaInputType=Bool
    rotationMode=BMM_Additive
    rotationSpace=BCS_BoneSpace
    translationMode=BMM_Additive
    translationSpace=BCS_BoneSpace
    alphaBoolBlend {blendInTime=0.2, blendOption=Linear, blendOutTime=0.2}
    boneToModify {boneName=base_bag_B}
    rotation=<rot -20.928 0 0>
    translation {y=-1.506}
   }
   n7 = "Transform (Modify) Bone - Bone: base_bag_A"(ComponentPose=@n6, bAlphaBoolEnabled=True, Translation="0.442,0.000,0.000", Rotation="0,0,0", Scale="1,1,1") {
    alphaInputType=Bool
    rotationMode=BMM_Additive
    rotationSpace=BCS_BoneSpace
    translationMode=BMM_Additive
    translationSpace=BCS_BoneSpace
    alphaBoolBlend {blendInTime=0.2, blendOption=Linear, blendOutTime=0.2}
    boneToModify {boneName=base_bag_A}
    translation {x=0.442}
   }
   n8 = "Transform (Modify) Bone - Bone: area_belt_fw"(ComponentPose=@n7, bAlphaBoolEnabled=True, Translation="3.462,-0.014,0.527", Rotation="0.000,0.000,-25.997", Scale="1,1,1") {
    alphaInputType=Bool
    rotationMode=BMM_Additive
    rotationSpace=BCS_BoneSpace
    translationMode=BMM_Additive
    translationSpace=BCS_BoneSpace
    alphaBoolBlend {blendInTime=0.1, blendOption=Linear, blendOutTime=0.2}
    boneToModify {boneName=area_belt_fw}
    rotation=<rot 0 0 -25.997>
    translation=<3.462 -0.014 0.527>
   }
   n9 = "Transform (Modify) Bone - Bone: base_grenade_A"!(ComponentPose=@n8, bAlphaBoolEnabled:bool=True, Translation="5,5,5", Rotation="0,0,5", Scale="1,1,1") -> (Pose) {
    alphaInputType=Bool
    rotationMode=BMM_Additive
    rotationSpace=BCS_BoneSpace
    translationMode=BMM_Additive
    translationSpace=BCS_BoneSpace
    alphaBoolBlend {blendInTime=0.1, blendOption=Linear, blendOutTime=0.2}
    boneToModify {boneName=base_grenade_A}
    rotation {roll=5}
    translation=<5 5 5>
   }
   Output Pose(Result=Component To Local(ComponentPose=@n9)) {layerGroup=None}
   n10 = Layered blend per bone(
    BasePose=Copy Pose From Mesh() {bUseAttachedParent=true}
    BlendPoses_0=Copy Pose From Mesh(SourceMeshComponent=Get Head() -> (Head:Skeletal Mesh Component Budgeted Object Reference)) {bUseAttachedParent=true}
   ) {layerSetup [{branchFilters [{boneName=head}, {boneName=head_twist}, {boneName=jaw}, {boneName=None}]}]}
   n11 = Control Rig!(
    Source=@n10
    Alpha
    Crouching:double=Get Alpha_Crouching() -> (Alpha_Crouching:double)
    DBNO:double=Get Alpha_DBNO() -> (Alpha_DBNO:double)
    ForwardDBNO:double=Get ForwardDBNO() -> (ForwardDBNO:double)
    Parachute:double=Get Parachute() -> (Parachute:double)
   ) -> (Pose) {
    alphaBoolBlend {blendOption=Linear}
    controlRigAssetReference {blueprintRigClass=&/M_MED_VelvetDeskCam_CtrlRig.M_MED_VelvetDeskCam_CtrlRig_C}
   }
   Save cached pose 'Ctrl Rig Cache'(Pose=@n11) {cacheName=Ctrl Rig Cache}
   n12 = Anim Dynamics - Bone: dyn_sleeves_fw_r(ComponentPose=Local To Component(LocalPose=Use cached pose 'Ctrl Rig Cache'()), Alpha=1, ExternalForce="0,0,0") {
    angularDampingOverride=1
    angularSpringConstant=20000
    bAngularSpring=true
    bOverrideAngularDamping=true
    bUseGravityOverride=true
    gravityScale=0.35
    simulationSpace=BoneRelative
    alphaBoolBlend {blendOption=Linear}
    boundBone {boneName=dyn_sleeves_fw_r}
    componentLinearAccScale=<1 1 1>
    componentLinearVelScale=<1 1 1>
    gravityOverride {z=-980}
    physicsBodyDefinitions [{
      boundBone {boneName=dyn_sleeves_fw_r}
      boxExtents=<20 30 30>
      constraintSetup {angularLimitsMax=<15 15 15>, angularLimitsMin=<-15 -15 -15>, angularTarget=<1 0 0>}
      localJointOffset=<4 1 0>
     }]
    relativeSpaceBone {boneName=pelvis}
   }
   n13 = Anim Dynamics - Bone: dyn_sleeves_dn_r(ComponentPose=@n12, Alpha=1, ExternalForce="0,0,0") {
    angularDampingOverride=1
    angularSpringConstant=20000
    bAngularSpring=true
    bOverrideAngularDamping=true
    bUseGravityOverride=true
    gravityScale=0.35
    simulationSpace=BoneRelative
    alphaBoolBlend {blendOption=Linear}
    boundBone {boneName=dyn_sleeves_dn_r}
    componentLinearAccScale=<1 1 1>
    componentLinearVelScale=<1 1 1>
    gravityOverride {z=-980}
    physicsBodyDefinitions [{
      boundBone {boneName=dyn_sleeves_dn_r}
      boxExtents=<20 30 30>
      constraintSetup {angularLimitsMax=<15 15 15>, angularLimitsMin=<-15 -15 -15>, angularTarget=<1 0 0>}
      localJointOffset=<4 1 0>
     }]
    relativeSpaceBone {boneName=pelvis}
   }
   n14 = Anim Dynamics - Bone: dyn_sleeves_dn_l(ComponentPose=@n13, Alpha=1, ExternalForce="0,0,0") {
    angularDampingOverride=1
    angularSpringConstant=20000
    bAngularSpring=true
    bOverrideAngularDamping=true
    bUseGravityOverride=true
    gravityScale=0.35
    simulationSpace=BoneRelative
    alphaBoolBlend {blendOption=Linear}
    boundBone {boneName=dyn_sleeves_dn_l}
    componentLinearAccScale=<1 1 1>
    componentLinearVelScale=<1 1 1>
    gravityOverride {z=-980}
    physicsBodyDefinitions [{
      boundBone {boneName=dyn_sleeves_dn_l}
      boxExtents=<20 30 30>
      constraintSetup {angularLimitsMax=<15 15 15>, angularLimitsMin=<-15 -15 -15>, angularTarget=<1 0 0>}
      localJointOffset=<4 1 0>
     }]
    relativeSpaceBone {boneName=pelvis}
   }
   n15 = Anim Dynamics - Bone: dyn_sleeves_fw_l(ComponentPose=@n14, Alpha=1, ExternalForce="0,0,0") {
    angularDampingOverride=1
    angularSpringConstant=20000
    bAngularSpring=true
    bOverrideAngularDamping=true
    bUseGravityOverride=true
    gravityScale=0.35
    simulationSpace=BoneRelative
    alphaBoolBlend {blendOption=Linear}
    boundBone {boneName=dyn_sleeves_fw_l}
    componentLinearAccScale=<1 1 1>
    componentLinearVelScale=<1 1 1>
    gravityOverride {z=-980}
    physicsBodyDefinitions [{
      boundBone {boneName=dyn_sleeves_fw_l}
      boxExtents=<20 30 30>
      constraintSetup {angularLimitsMax=<15 15 15>, angularLimitsMin=<-15 -15 -15>, angularTarget=<1 0 0>}
      localJointOffset=<4 1 0>
     }]
    relativeSpaceBone {boneName=pelvis}
   }
   n16 = Anim Dynamics - Bone: dyn_bag_A(ComponentPose=@n15, Alpha=0.65, ExternalForce="0,0,0") {
    angularDampingOverride=1
    angularSpringConstant=20000
    bAngularSpring=true
    bOverrideAngularDamping=true
    bUseGravityOverride=true
    gravityScale=0.35
    simulationSpace=BoneRelative
    alphaBoolBlend {blendOption=Linear}
    boundBone {boneName=dyn_bag_A}
    componentLinearAccScale=<0.5 0.5 0.5>
    componentLinearVelScale=<0.5 0.5 0.5>
    gravityOverride {z=-980}
    physicsBodyDefinitions [{
      boundBone {boneName=dyn_bag_A}
      boxExtents=<20 30 30>
      constraintSetup {angularLimitsMax=<2 0 5>, angularLimitsMin=<0 -15 -5>, angularTarget=<1 0 0>}
      localJointOffset=<3 0 0>
     }]
   }
   n17 = Anim Dynamics - Bone: dyn_bag_B(ComponentPose=@n16, Alpha=0.65, ExternalForce="0,0,0") {
    angularDampingOverride=1
    angularSpringConstant=20000
    bAngularSpring=true
    bOverrideAngularDamping=true
    bUseGravityOverride=true
    gravityScale=0.35
    simulationSpace=BoneRelative
    alphaBoolBlend {blendOption=Linear}
    boundBone {boneName=dyn_bag_B}
    componentLinearAccScale=<0.5 0.5 0.5>
    componentLinearVelScale=<0.5 0.5 0.5>
    gravityOverride {z=-980}
    physicsBodyDefinitions [{
      boundBone {boneName=dyn_bag_B}
      boxExtents=<20 30 30>
      constraintSetup {angularLimitsMax=<0 5 7>, angularLimitsMin=<-1 -10 -7>, angularTarget=<1 0 0>}
      localJointOffset=<-1 0 0>
     }]
   }
   n18 = Anim Dynamics - Bone: dyn_belt_area(ComponentPose=@n17, Alpha=0, ExternalForce="0,0,0") {
    angularDampingOverride=1
    angularSpringConstant=20000
    bAngularSpring=true
    bOverrideAngularDamping=true
    bUseGravityOverride=true
    gravityScale=0.35
    simulationSpace=BoneRelative
    alphaBoolBlend {blendOption=Linear}
    boundBone {boneName=dyn_belt_area}
    componentLinearAccScale=<0.8 0.8 0.8>
    componentLinearVelScale=<0.8 0.8 0.8>
    gravityOverride {z=-980}
    physicsBodyDefinitions [{
      boundBone {boneName=dyn_belt_area}
      boxExtents=<20 30 30>
      constraintSetup {angularLimitsMax=<2 5 0>, angularLimitsMin=<-2 0 0>, angularTarget=<1 0 0>}
      localJointOffset=<4 0 0>
     }]
    relativeSpaceBone {boneName=pelvis}
   }
   n19 = Get GravityOverride()
   n20 = Anim Dynamics - Bone: dyn_grenade_C_loop!(ComponentPose=@n18, Alpha=1, GravityOverride:Vector=@n19, ExternalForce="0,0,0") -> (Pose) {
    angularDampingOverride=1
    angularSpringConstant=500
    bAngularSpring=true
    bOverrideAngularDamping=true
    bUseGravityOverride=true
    gravityScale=0.35
    simulationSpace=BoneRelative
    alphaBoolBlend {blendOption=Linear}
    boundBone {boneName=dyn_grenade_C_loop}
    componentAppliedLinearAccClamp=<1000 1000 1000>
    componentLinearAccScale=<0.3 0.3 0.3>
    componentLinearVelScale=<0.3 0.3 0.3>
    gravityOverride {z=-980}
    physicsBodyDefinitions [{
      boundBone {boneName=dyn_grenade_C_loop}
      boxExtents=<4 4 4>
      constraintSetup {twistAxis=AxisZ, angularLimitsMax=<80 90 15>, angularLimitsMin=<-80 -40 -15>, angularTarget=<1 0 0>}
      localJointOffset=<0 0 3>
     }]
    relativeSpaceBone {boneName=base_grenade_C}
   }
   n21 = Anim Dynamics - Bone: dyn_grenade_B_loop!(ComponentPose=@n20, Alpha=1, GravityOverride:Vector=@n19, ExternalForce="0,0,0") -> (Pose) {
    angularDampingOverride=1
    angularSpringConstant=500
    bAngularSpring=true
    bOverrideAngularDamping=true
    bUseGravityOverride=true
    gravityScale=0.35
    simulationSpace=BoneRelative
    alphaBoolBlend {blendOption=Linear}
    boundBone {boneName=dyn_grenade_B_loop}
    componentAppliedLinearAccClamp=<500 500 500>
    componentLinearAccScale=<0.3 0.3 0.3>
    componentLinearVelScale=<0.3 0.3 0.3>
    gravityOverride {z=-980}
    physicsBodyDefinitions [{
      boundBone {boneName=dyn_grenade_B_loop}
      boxExtents=<4 4 4>
      constraintSetup {twistAxis=AxisZ, angularLimitsMax=<80 90 15>, angularLimitsMin=<-80 -40 -15>, angularTarget=<1 0 0>}
      localJointOffset=<0 0 3>
     }]
    relativeSpaceBone {boneName=base_grenade_B}
   }
   n22 = Anim Dynamics - Bone: dyn_grenade_A_loop!(ComponentPose=@n21, Alpha=1, GravityOverride:Vector=@n19, ExternalForce="0,0,0") -> (Pose) {
    angularDampingOverride=1
    angularSpringConstant=50
    bAngularSpring=true
    bOverrideAngularDamping=true
    bUseGravityOverride=true
    gravityScale=0.35
    simulationSpace=BoneRelative
    alphaBoolBlend {blendOption=Linear}
    boundBone {boneName=dyn_grenade_A_loop}
    componentAppliedLinearAccClamp=<1000 1000 1000>
    componentLinearAccScale=<0.3 0.3 0.3>
    componentLinearVelScale=<0.3 0.3 0.3>
    gravityOverride {z=-980}
    physicsBodyDefinitions [{
      boundBone {boneName=dyn_grenade_A_loop}
      boxExtents=<4 4 4>
      constraintSetup {twistAxis=AxisZ, angularLimitsMax=<80 90 15>, angularLimitsMin=<-80 -40 -15>, angularTarget=<1 0 0>}
      localJointOffset=<0 0 3>
     }]
    relativeSpaceBone {boneName=base_grenade_A}
   }
   n23 = Anim Dynamics - Bone: dyn_bag_B_loop(ComponentPose=@n22, Alpha=1, ExternalForce="0,0,0") {
    angularDampingOverride=1
    angularSpringConstant=20000
    bOverrideAngularDamping=true
    gravityScale=0.35
    simulationSpace=BoneRelative
    alphaBoolBlend {blendOption=Linear}
    boundBone {boneName=dyn_bag_B_loop}
    componentAppliedLinearAccClamp=<5000 5000 5000>
    componentLinearAccScale=<0.3 0.3 0.3>
    componentLinearVelScale=<0.3 0.3 0.3>
    gravityOverride {z=-980}
    physicsBodyDefinitions [{
      boundBone {boneName=dyn_bag_B_loop}
      boxExtents=<10 10 10>
      constraintSetup {twistAxis=AxisZ, angularLimitsMax=<15 15 15>, angularLimitsMin=<-80 -15 -15>, angularTarget=<1 0 0>}
      localJointOffset=<0 0 2>
     }]
   }
   n24 = Anim Dynamics - Bone: dyn_bag_A_loop(ComponentPose=@n23, Alpha=1, ExternalForce="0,0,0") {
    angularDampingOverride=1
    angularSpringConstant=20000
    bOverrideAngularDamping=true
    gravityScale=0.35
    simulationSpace=BoneRelative
    alphaBoolBlend {blendOption=Linear}
    boundBone {boneName=dyn_bag_A_loop}
    componentAppliedLinearAccClamp=<5000 5000 5000>
    componentLinearAccScale=<0.3 0.3 0.3>
    componentLinearVelScale=<0.3 0.3 0.3>
    gravityOverride {z=-980}
    physicsBodyDefinitions [{
      boundBone {boneName=dyn_bag_A_loop}
      boxExtents=<10 10 10>
      constraintSetup {twistAxis=AxisZ, angularLimitsMax=<80 15 15>, angularLimitsMin=<-15 -15 -15>, angularTarget=<1 0 0>}
      localJointOffset=<0 0 2>
     }]
   }
   n25 = Get IsNOTTargeting() -> (IsNOTTargeting:double)
   n26 = Get GravityOverride()
   n27 = Anim Dynamics - Bone: dyn_grenade_C!(ComponentPose=@n24, Alpha=@n25, GravityOverride:Vector=@n26, ExternalForce="0,0,0") -> (Pose) {
    angularDampingOverride=1
    angularSpringConstant=5000
    bAngularSpring=true
    bOverrideAngularDamping=true
    bUseGravityOverride=true
    gravityScale=0.35
    simulationSpace=RootRelative
    alphaBoolBlend {blendOption=Linear}
    alphaScaleBiasClamp {bInterpResult=true}
    boundBone {boneName=dyn_grenade_C}
    componentLinearAccScale=<1 1 1>
    componentLinearVelScale=<1 1 1>
    gravityOverride {z=-980}
    physicsBodyDefinitions [{
      boundBone {boneName=dyn_grenade_C}
      boxExtents=<5 10 10>
      constraintSetup {angularLimitsMax=<15 2 0>, angularLimitsMin=<-15 -2 -15>, angularTarget=<1 0 0>}
      localJointOffset=<2 0 0>
     }]
   }
   n28 = Anim Dynamics - Bone: dyn_grenade_B!(ComponentPose=@n27, Alpha=@n25, GravityOverride:Vector=@n26, ExternalForce="0,0,0") -> (Pose) {
    angularDampingOverride=1
    angularSpringConstant=5000
    bAngularSpring=true
    bOverrideAngularDamping=true
    bUseGravityOverride=true
    gravityScale=0.35
    simulationSpace=RootRelative
    alphaBoolBlend {blendOption=Linear}
    alphaScaleBiasClamp {bInterpResult=true}
    boundBone {boneName=dyn_grenade_B}
    componentLinearAccScale=<1 1 1>
    componentLinearVelScale=<1 1 1>
    gravityOverride {z=-980}
    physicsBodyDefinitions [{
      boundBone {boneName=dyn_grenade_B}
      boxExtents=<5 10 10>
      constraintSetup {angularLimitsMax=<15 2 5>, angularLimitsMin=<-15 -2 -5>, angularTarget=<1 0 0>}
      localJointOffset=<2 0 0>
     }]
   }
   n29 = Anim Dynamics - Bone: dyn_grenade_A!(ComponentPose=@n28, Alpha=@n25, GravityOverride:Vector=@n26, ExternalForce="0,0,0") -> (Pose) {
    angularDampingOverride=1
    angularSpringConstant=5000
    bAngularSpring=true
    bOverrideAngularDamping=true
    bUseGravityOverride=true
    gravityScale=0.35
    simulationSpace=RootRelative
    alphaBoolBlend {blendOption=Linear}
    alphaScaleBiasClamp {bInterpResult=true}
    boundBone {boneName=dyn_grenade_A}
    componentLinearAccScale=<1 1 1>
    componentLinearVelScale=<1 1 1>
    gravityOverride {z=-980}
    physicsBodyDefinitions [{
      boundBone {boneName=dyn_grenade_A}
      boxExtents=<5 10 10>
      constraintSetup {angularLimitsMax=<15 2 0>, angularLimitsMin=<-15 -2 0>, angularTarget=<1 0 0>}
      localJointOffset=<2 0 0>
     }]
   }
   n30 = "Transform (Modify) Bone - Bone: FACIAL_C_Jaw"(ComponentPose=@n29, Alpha=Get Neck_Alpha() -> (Neck_Alpha:double), Translation="0,0,3", Rotation="0,0,0", Scale="1.2,1.2,1.2") {
    scaleMode=BMM_Additive
    scaleSpace=BCS_BoneSpace
    translationMode=BMM_Additive
    translationSpace=BCS_BoneSpace
    alphaBoolBlend {blendOption=Linear}
    alphaScaleBiasClamp {bInterpResult=true}
    boneToModify {boneName=FACIAL_C_Jaw}
    scale=<1.2 1.2 1.2>
    translation {z=3}
   }
   n31 = "Transform (Modify) Bone - Bone: neck_correct"(ComponentPose=@n30, Alpha=Get PlayingCelebration() -> (PlayingCelebration:double), Translation="1.2,0,0", Rotation="0,0,0", Scale="1,1,1") {
    scaleMode=BMM_Additive
    scaleSpace=BCS_BoneSpace
    translationMode=BMM_Additive
    translationSpace=BCS_BoneSpace
    alphaBoolBlend {blendOption=Linear}
    alphaScaleBiasClamp {bInterpResult=true}
    boneToModify {boneName=neck_correct}
    translation {x=1.2}
   }
   n32 = "Transform (Modify) Bone - Bone: neck_01"(ComponentPose=@n31, Alpha=Get Lower_LOD() -> (Lower_LOD:double), Translation="0,0,0", Rotation="0,0,0", Scale="1.1,1.1,1.11") {
    scaleMode=BMM_Additive
    scaleSpace=BCS_BoneSpace
    translationMode=BMM_Additive
    translationSpace=BCS_BoneSpace
    alphaBoolBlend {blendOption=Linear}
    alphaScaleBiasClamp {bInterpResult=true}
    boneToModify {boneName=neck_01}
    scale=<1.1 1.1 1.11>
   }
   Save cached pose 'Dynamics Cache'(Pose=Component To Local(ComponentPose=@n32)) {cacheName=Dynamics Cache}
  ]}
 EventGraph {nodes [
   Event BlueprintUpdateAnimation()
   > n1 = Sequence!(execute) -> (then_0, then_1, then_2:Exec, then_3:Exec, then_4:Exec, then_5:Exec, then_6:Exec)
   > n2 = Gravity Override State Function DEPRECATED(CharacterPartAnimInstance=Self-Reference(), GameTimeInSeconds=GetGameTimeInSeconds(), GravityOverrideParams=Get Gravity Override Params())
   > Set GravityOverride(GravityOverride=@n2)
   > n3 = Branch(Condition=Get bIsCrouching())
   > Set Alpha_Crouching(Alpha_Crouching:double=1) -> (Output_Get:double)
   Set Alpha_Crouching(execute=@n3.else, Alpha_Crouching=0)
   n4 = Branch(execute=@n1.then_1, Condition=Get bIsDBNO())
   > Set Alpha_DBNO(Alpha_DBNO:double=1) -> (Output_Get:double)
   Set Alpha_DBNO(execute=@n4.else, Alpha_DBNO=0)
   n5 = Get bIsParachuteOpened()
   n6 = Get bIsSkydiveFloating()
   Set Parachute(execute=@n1.then_3, Parachute:double="To Float (Boolean)"(InBool=OR Boolean(A=@n5, B=@n6))) -> (Output_Get:double)
   n7 = NOT Boolean(A=Get bIsTargeting())
   n8 = NOT Boolean(A=Get bIsPlayingEmote())
   n9 = NOT Boolean(A=Get bIsPlayingMeleeAnim())
   > Set IsNOTTargeting(IsNOTTargeting:double="To Float (Boolean)"(InBool=AND Boolean!(A=@n7, B=@n9, C:bool=@n8) -> (ReturnValue))) -> (Output_Get:double)
   n10 = Get CurrentLOD()
   > Set Lower_LOD(Lower_LOD:double="To Float (Boolean)"(InBool="integer >= integer"(A=@n10, B=1))) -> (Output_Get:double)
   n11 = Get bIsPlayingEmote()
   > Set FrontEndIdle(FrontEndIdle:bool=AND Boolean!(A=Get bIsInFrontEnd(), B=NOT Boolean(A=@n11)) -> (ReturnValue)) -> (Output_Get:bool)
   n12 = GetOwningActor()
   Cast To FortPlayerMannequin!(execute=@n1.then_4, Object=@n12) -> (then, CastFailed, AsFort Player Mannequin:Fort Player Mannequin Object Reference) {is=title K2Node_DynamicCast}
   > n13 = For Each Loop(Array:Array of Skeletal Mesh Component Object References=Get Components By Class!(self:Actor Object Reference=@n12, ComponentClass:Actor Component Class Reference=/Script/Engine.SkeletalMeshComponent) -> (ReturnValue:Array of Skeletal Mesh Component Object References)) -> (Array Element:Skeletal Mesh Component Object Reference)
   > Branch(Condition="Equal (Object)"(
     A=Get SkeletalMesh!(self:Skinned Mesh Component Object Reference=@n13) -> (SkeletalMesh:Skeletal Mesh Object Reference)
     B=/BRCosmetics/Characters/Player/Male/Medium/Heads/M_MED_VelvetDeskCam_Head/Meshes/M_MED_VelvetDeskCam_Head
    ))
   > n14 = Cast To SkeletalMeshComponentBudgeted!(execute, Object=@n13) -> (then, CastFailed, AsSkeletal Mesh Component Budgeted:Skeletal Mesh Component Budgeted Object Reference) {is=title K2Node_DynamicCast}
   > Set Head(Head:Skeletal Mesh Component Budgeted Object Reference=@n14) -> (Output_Get:Skeletal Mesh Component Budgeted Object Reference)
   n15 = GetAnimInstance(self=GetOwnerSkeletalMeshComponent())
   Set PlayingCelebration(
    execute=@n1.then_6
    PlayingCelebration:double="To Float (Boolean)"(InBool=Montage_IsPlaying(self=@n15, Montage=/Game/Animation/Game/MainPlayer/Emotes/Emote_Celebration_M))
   ) -> (Output_Get:double)
   n16 = PlayerMovement(execute=@n1.then_2, AnimInstance=Self-Reference(), XDotProductCoefficient=100, YDotProductCoefficient=100, ZDotProductCoefficient=100)
   n17 = "float > float"(A=BreakVector(InVec=MakeVector(X=@n16, Y=@n16.YMovement)).Y, B=40)
   n18 = Get bIsDBNO()
   > Set ForwardDBNO(ForwardDBNO:double="To Float (Boolean)"(InBool=AND Boolean(A=@n17, B=@n18))) -> (Output_Get:double)
   n19 = BreakTransform(InTransform=GetDeltaTransformFromRefPose(self=Get BodyMesh() -> (BodyMesh:Skeletal Mesh Component Object Reference), BoneName=clavicle_l, BaseName=spine_05))
   Set Neck_Alpha(execute=@n1.then_5, Neck_Alpha:double=MapRangeClamped(Value=BreakRotator(InRot=@n19.Rotation), InRangeA=0, InRangeB=15)) -> (Output_Get:double)
   n20 = Event OnCharacterCustomizationLoaded()
   > Set BodyMesh(BodyMesh=@n20.CustomBaseMesh)
   GetOwnerSkeletalMeshComponent()
  ]}
]
variables [
 Alpha_Crouching {type=number}
 Alpha_Crouching Moving2D {type=number}
 Alpha_DBNO {type=number}
 BodyMesh {type=SkeletalMeshComponent}
 FACIAL_C_Chin_Rotation {type=Rotator}
 FACIAL_C_Chin_Translation {type=Vector}
 FACIAL_C_Jaw_Rotation {type=Rotator}
 FACIAL_C_Jaw_Tranform {type=Vector}
 FACIAL_C_Jawline_Rotation {type=Rotator}
 FACIAL_C_Jawline_Translation {type=Vector}
 FACIAL_L_UnderChin_Rotation {type=Rotator}
 FACIAL_L_UnderChin_Translation {type=Vector}
 FACIAL_R_UnderChin_Rotation {type=Rotator}
 FACIAL_R_UnderChin_Translation {type=Vector}
 ForwardDBNO {type=number}
 FrontEndIdle {type=boolean}
 Gravity Override Params {type=GravityOverrideParamsStruct, default {rail_Grinding {gravityOverride=<0 0 -2000>}}}
 GravityOverride {type=Vector}
 Head {type=SkeletalMeshComponentBudgeted}
 HeadMesh {type=SkeletalMeshComponentBudgeted}
 IsNOTTargeting {type=number}
 isUsingFacialBones {type=boolean}
 Lower_LOD {type=number}
 Neck_Alpha {type=number}
 Parachute {type=number}
 PlayingCelebration {type=number}
 Swimote {type=number}
]
